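# Reusable icon button Editor Utility Widget in the VO Utilities audio tools, named for the Fort Sound Sequence editor.
# A styled Editor Utility Button inside a SizeBox holds a grey-tinted icon image, with a rounded background on hover and press.
# PreConstruct applies the ToolTip text and builds the image brush from the Icon texture and IconSize variables.
# Pressing the button broadcasts the On Button Pressed dispatcher.
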EditorUtilityWidgetBlueprint /Game/Sounds/Fort_Audio_Blueprints/Utilities/EditorUtilityWidgets/VO_Utilities/Widgets/FortSoundSequenceEditorButton
graphs [
 EventGraph {nodes [
   "On Pressed (Button)"() {is=title K2Node_ComponentBoundEvent}
   > Call On Button Pressed!(execute:Exec, self:Fort Sound Sequence Editor Button Object Reference) -> (then:Exec) {is=title K2Node_CallDelegate}
   Event PreConstruct()
   > SetToolTipText!(
    execute
    self:Widget Object Reference=Get Button() -> (Button:Editor Utility Button Object Reference)
    InToolTipText:"Text (by ref)"=Get ToolTip() -> (ToolTip:Text)
   ) -> (then)
   > SetBrush(
    self=Get IconImage() -> (IconImage:Image Object Reference)
    InBrush=Make Slate Brush!(
     TintColor:Slate Color Structure=(SpecifiedColor=(R=1,G=1,B=1,A=1),ColorUseRule=UseColor_Specified)
     DrawAs:ESlateBrushDrawType Enum=Image
     Tiling:ESlateBrushTileType Enum=NoTile
     Mirroring:ESlateBrushMirrorType Enum=NoMirror
     ImageSize:Vector 2D Structure=Get IconSize() -> (IconSize:Vector 2D Structure)
     Margin:Margin Structure=(Left=0,Top=0,Right=0,Bottom=0)
     ResourceObject:Object Reference=Get Icon() -> (Icon:Texture Object Reference)
     OutlineSettings:Slate Brush Outline Settings Structure=(CornerRadii=(X=0,Y=0,Z=0,W=0),Color=(SpecifiedColor=(R=0,G=0,B=0,A=0),ColorUseRule=UseColor_Specified),Width=0,RoundingType=HalfHeightRadius,bUseBrushTransparency=False)
    ) -> (SlateBrush:Slate Brush Structure) {is=title K2Node_MakeStruct}
   )
  ]}
 On Button Pressed {nodes [On Button Pressed() {is=title K2Node_FunctionEntry}]}
]
variables [
 Icon {type=Texture, default=/Game/UI/Foundation/Textures/Icons/Stars/Stars-Full-64}
 IconSize {type=Vector2D, default=<24 24>}
 On Button Pressed {default=()}
 ToolTip {type=string, default=""}
]
widgets {
 info {rootWidgetClass=&SizeBox, widgetCount=3}
 widgets [
  SizeBox_1 {widgetClassPath=&SizeBox, widget=SizeBox_1, properties {bOverride_MaxAspectRatio=true, bOverride_MinAspectRatio=true}}
  Button {
   widgetClassPath=&EditorUtilityButton
   parent=SizeBox_1
   widget=Button
   properties {widgetStyle {
     clickedSlateSound {resourceObject=null}
     disabled {
      bIsDynamicallyLoaded=false
      drawAs=RoundedBox
      imageType=NoImage
      mirroring=NoMirror
      resourceName=None
      resourceObject=null
      tiling=NoTile
      imageSize=<0 0>
      margin=<ltrb 0 0 0 0>
      outlineSettings {
       bUseBrushTransparency=false
       roundingType=FixedRadius
       width=1
       color {colorUseRule=UseColor_ColorTable, specifiedColor=<rgba 0 0 0 0>}
       cornerRadii=<4 4 4 4>
      }
      tintColor {colorUseRule=UseColor_Specified, specifiedColor=<rgba 0.0395 0.0395 0.0395 1>}
      uVRegion {bIsValid=false, max=<0 0>, min=<0 0>}
     }
     disabledForeground {colorUseRule=UseColor_ColorTable, specifiedColor=<rgba 0 0 0 0>}
     hovered {
      bIsDynamicallyLoaded=false
      drawAs=RoundedBox
      imageType=NoImage
      mirroring=NoMirror
      resourceName=None
      resourceObject=null
      tiling=NoTile
      imageSize=<0 0>
      margin=<ltrb 0 0 0 0>
      outlineSettings {
       bUseBrushTransparency=false
       roundingType=FixedRadius
       width=0
       color {colorUseRule=UseColor_Foreground, specifiedColor=<rgba 0 0 0 0>}
       cornerRadii=<8 8 8 8>
      }
      tintColor {colorUseRule=UseColor_Specified, specifiedColor=<rgba 0.0103 0.0103 0.0103 1>}
      uVRegion {bIsValid=false, max=<0 0>, min=<0 0>}
     }
     hoveredForeground {colorUseRule=UseColor_ColorTable, specifiedColor=<rgba 0 0 0 0>}
     hoveredSlateSound {resourceObject=null}
     normal {
      bIsDynamicallyLoaded=false
      drawAs=NoDrawType
      imageType=NoImage
      mirroring=NoMirror
      resourceName=None
      resourceObject=null
      tiling=NoTile
      imageSize=<0 0>
      margin=<ltrb 0 0 0 0>
      outlineSettings {
       bUseBrushTransparency=false
       roundingType=FixedRadius
       width=0
       color {colorUseRule=UseColor_Foreground, specifiedColor=<rgba 0 0 0 0>}
       cornerRadii=<4 4 4 4>
      }
      tintColor {colorUseRule=UseColor_Specified, specifiedColor=<rgba 0.0312 0.0312 0.0312 1>}
      uVRegion {bIsValid=false, max=<0 0>, min=<0 0>}
     }
     normalForeground {colorUseRule=UseColor_ColorTable, specifiedColor=<rgba 0 0 0 0>}
     normalPadding=<ltrb 2 2 2 2>
     pressed {
      bIsDynamicallyLoaded=false
      drawAs=RoundedBox
      imageType=NoImage
      mirroring=NoMirror
      resourceName=None
      resourceObject=null
      tiling=NoTile
      imageSize=<0 0>
      margin=<ltrb 0 0 0 0>
      outlineSettings {
       bUseBrushTransparency=false
       roundingType=FixedRadius
       width=0
       color {colorUseRule=UseColor_Foreground, specifiedColor=<rgba 0 0 0 0>}
       cornerRadii=<8 8 8 8>
      }
      tintColor {colorUseRule=UseColor_Specified, specifiedColor=<rgba 0.0052 0.0052 0.0052 1>}
      uVRegion {bIsValid=false, max=<0 0>, min=<0 0>}
     }
     pressedForeground {colorUseRule=UseColor_ColorTable, specifiedColor=<rgba 0 0 0 0>}
     pressedPadding=<ltrb 2 2 2 2>
     pressedSlateSound {resourceObject=null}
    }}
   slot {class=SizeBoxSlot, properties {horizontalAlignment=HAlign_Left, verticalAlignment=VAlign_Top}}
  }
  IconImage {
   widgetClassPath=&Image
   bIsVariable=true
   parent=Button
   widget=IconImage
   properties {
    brush {
     bIsDynamicallyLoaded=false
     drawAs=Image
     imageType=FullColor
     mirroring=NoMirror
     resourceName=None
     resourceObject=&Stars-Full-64
     tiling=NoTile
     imageSize=<24 24>
     margin=<ltrb 0 0 0 0>
     outlineSettings {
      bUseBrushTransparency=false
      roundingType=HalfHeightRadius
      width=0
      color {colorUseRule=UseColor_Specified, specifiedColor=<rgba 0 0 0 0>}
      cornerRadii=<0 0 0 0>
     }
     tintColor {colorUseRule=UseColor_Specified, specifiedColor=<rgba 1 1 1 1>}
     uVRegion {bIsValid=false, max=<0 0>, min=<0 0>}
    }
    colorAndOpacity=<rgba 0.5029 0.5029 0.5029 1>
   }
   slot {class=ButtonSlot, properties {padding=<ltrb 2 2 2 2>}}
  }
 ]
}
