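# Editor utility widget for a reusable icon toggle button, named for the Fort Sound Sequence Editor among the VO audio utilities.
# On PreConstruct, it builds the image brush from the Icon and IconSize variables and sets the toggle-styled checkbox from Is Enabled by Default.
# Checkbox changes are re-broadcast through the On Check State Changed dispatcher.
# Get Icon Color returns white when checked and grey when unchecked.
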EditorUtilityWidgetBlueprint /Game/Sounds/Fort_Audio_Blueprints/Utilities/EditorUtilityWidgets/VO_Utilities/Widgets/FortSoundSequenceEditorToggleButton
graphs [
 EventGraph {nodes [
   n1 = "On Check State Changed (ToggleButton)"!() -> (OutputDelegate, then, bIsChecked:bool) {is=title K2Node_ComponentBoundEvent}
   > Call On Check State Changed!(execute:Exec, self:Fort Sound Sequence Editor Toggle Button Object Reference, Is Checked:bool=@n1.bIsChecked) -> (then:Exec) {is=title K2Node_CallDelegate}
   Event PreConstruct()
   > SetBrush(
    self=Get IconImage() -> (IconImage:Image Object Reference)
    InBrush=Get Icon!() -> (ReturnValue:Slate Brush Structure) {is=title K2Node_CallFunction}
   )
   > SetCheckedState!(
    execute
    self:Check Box Object Reference=Get ToggleButton() -> (ToggleButton:Editor Utility Check Box Object Reference)
    InCheckedState:ECheckBoxState Enum=Select(
     Option 0:ECheckBoxState Enum=Unchecked
     Option 1:ECheckBoxState Enum=Checked
     Index=Get Is Enabled by Default() -> (Is Enabled by Default:bool)
    ) -> (ReturnValue:ECheckBoxState Enum)
   ) -> (then)
  ]}
 Get Icon {nodes [
   Get Icon() {is=title K2Node_FunctionEntry}
   > Return Node!(
    execute
    ReturnValue:Slate Brush Structure=Make Slate Brush!(
     TintColor:Slate Color Structure=(SpecifiedColor=(R=1,G=1,B=1,A=1),ColorUseRule=UseColor_Specified)
     DrawAs:ESlateBrushDrawType Enum=Image
     Tiling:ESlateBrushTileType Enum=NoTile
     Mirroring:ESlateBrushMirrorType Enum=NoMirror
     ImageSize:Vector 2D Structure=Get IconSize() -> (IconSize:Vector 2D Structure)
     Margin:Margin Structure=(Left=0,Top=0,Right=0,Bottom=0)
     ResourceObject:Object Reference=Get Icon() -> (Icon:Texture Object Reference) {is=title K2Node_VariableGet}
     OutlineSettings:Slate Brush Outline Settings Structure=(CornerRadii=(X=0,Y=0,Z=0,W=0),Color=(SpecifiedColor=(R=0,G=0,B=0,A=0),ColorUseRule=UseColor_Specified),Width=0,RoundingType=HalfHeightRadius,bUseBrushTransparency=False)
    ) -> (SlateBrush:Slate Brush Structure) {is=title K2Node_MakeStruct}
   )
  ]}
 Get Icon Color {nodes [
   Get Icon Color() {is=title K2Node_FunctionEntry}
   n1 = Get ToggleButton()
   > Return Node(ReturnValue:Linear Color Structure=Select(
     Option 0:Linear Color Structure=(R=0.5029,G=0.5029,B=0.5029,A=1)
     Option 1:Linear Color Structure=(R=1,G=1,B=1,A=1)
     Index=IsChecked(self=@n1)
    ) -> (ReturnValue:Linear Color Structure))
  ]}
 On Check State Changed {nodes [On Check State Changed!() -> (then, Is Checked:bool=~) {is=title K2Node_FunctionEntry}]}
]
variables [
 Icon {type=Texture, default=/Game/UI/Foundation/Textures/Icons/Stars/Stars-Full-64}
 IconSize {type=Vector2D, default=<24 24>}
 Is Enabled by Default {type=boolean}
 On Check State Changed {default=()}
 ToolTip {type=string, default=""}
]
widgets {
 info {rootWidgetClass=&SizeBox, widgetCount=3}
 widgets [
  SizeBox_1 {widgetClassPath=&SizeBox, widget=SizeBox_1, properties {bOverride_MaxAspectRatio=true, bOverride_MinAspectRatio=true}}
  ToggleButton {
   widgetClassPath=&EditorUtilityCheckBox
   bIsVariable=true
   parent=SizeBox_1
   widget=ToggleButton
   properties {
    toolTipText=Enable Text Editing
    widgetStyle {
     checkBoxType=ToggleButton
     backgroundHoveredImage {
      bIsDynamicallyLoaded=false
      drawAs=RoundedBox
      imageType=NoImage
      mirroring=NoMirror
      resourceName=None
      resourceObject=null
      tiling=NoTile
      imageSize=<18 18>
      margin=<ltrb 0 0 0 0>
      outlineSettings {
       bUseBrushTransparency=false
       roundingType=FixedRadius
       width=1
       color {colorUseRule=UseColor_ColorTable, specifiedColor=<rgba 0 0 0 0>}
       cornerRadii=<3 3 3 3>
      }
      tintColor {colorUseRule=UseColor_Specified, specifiedColor=<rgba 0.0048 0.0048 0.0048 1>}
      uVRegion {bIsValid=false, max=<0 0>, min=<0 0>}
     }
     backgroundImage {
      bIsDynamicallyLoaded=false
      drawAs=RoundedBox
      imageType=NoImage
      mirroring=NoMirror
      resourceName=None
      resourceObject=null
      tiling=NoTile
      imageSize=<18 18>
      margin=<ltrb 0 0 0 0>
      outlineSettings {
       bUseBrushTransparency=false
       roundingType=FixedRadius
       width=1
       color {colorUseRule=UseColor_ColorTable, specifiedColor=<rgba 0 0 0 0>}
       cornerRadii=<3 3 3 3>
      }
      tintColor {colorUseRule=UseColor_Specified, specifiedColor=<rgba 0.0048 0.0048 0.0048 1>}
      uVRegion {bIsValid=false, max=<0 0>, min=<0 0>}
     }
     backgroundPressedImage {
      bIsDynamicallyLoaded=false
      drawAs=RoundedBox
      imageType=NoImage
      mirroring=NoMirror
      resourceName=None
      resourceObject=null
      tiling=NoTile
      imageSize=<18 18>
      margin=<ltrb 0 0 0 0>
      outlineSettings {
       bUseBrushTransparency=false
       roundingType=FixedRadius
       width=1
       color {colorUseRule=UseColor_ColorTable, specifiedColor=<rgba 0 0 0 0>}
       cornerRadii=<3 3 3 3>
      }
      tintColor {colorUseRule=UseColor_Specified, specifiedColor=<rgba 0.0048 0.0048 0.0048 1>}
      uVRegion {bIsValid=false, max=<0 0>, min=<0 0>}
     }
     borderBackgroundColor {colorUseRule=UseColor_Specified, specifiedColor=<rgba 1 1 1 1>}
     checkedForeground {colorUseRule=UseColor_Specified, specifiedColor=<rgba 1 1 1 1>}
     checkedHoveredForeground {colorUseRule=UseColor_Specified, specifiedColor=<rgba 1 1 1 1>}
     checkedHoveredImage {
      bIsDynamicallyLoaded=false
      drawAs=RoundedBox
      imageType=NoImage
      mirroring=NoMirror
      resourceName=None
      resourceObject=null
      tiling=NoTile
      imageSize=<0 0>
      margin=<ltrb 0 0 0 0>
      outlineSettings {
       bUseBrushTransparency=false
       roundingType=FixedRadius
       width=0
       color {colorUseRule=UseColor_Foreground, specifiedColor=<rgba 0 0 0 0>}
       cornerRadii=<8 8 8 8>
      }
      tintColor {colorUseRule=UseColor_Specified, specifiedColor=<rgba 0.0044 0.2384 1 1>}
      uVRegion {bIsValid=false, max=<0 0>, min=<0 0>}
     }
     checkedImage {
      bIsDynamicallyLoaded=false
      drawAs=RoundedBox
      imageType=NoImage
      mirroring=NoMirror
      resourceName=None
      resourceObject=null
      tiling=NoTile
      imageSize=<0 0>
      margin=<ltrb 0 0 0 0>
      outlineSettings {
       bUseBrushTransparency=false
       roundingType=FixedRadius
       width=0
       color {colorUseRule=UseColor_Foreground, specifiedColor=<rgba 0 0 0 0>}
       cornerRadii=<8 8 8 8>
      }
      tintColor {colorUseRule=UseColor_Specified, specifiedColor=<rgba 0 0.162 0.7454 1>}
      uVRegion {bIsValid=false, max=<0 0>, min=<0 0>}
     }
     checkedPressedForeground {colorUseRule=UseColor_Specified, specifiedColor=<rgba 1 1 1 1>}
     checkedPressedImage {
      bIsDynamicallyLoaded=false
      drawAs=RoundedBox
      imageType=NoImage
      mirroring=NoMirror
      resourceName=None
      resourceObject=null
      tiling=NoTile
      imageSize=<0 0>
      margin=<ltrb 0 0 0 0>
      outlineSettings {
       bUseBrushTransparency=false
       roundingType=FixedRadius
       width=0
       color {colorUseRule=UseColor_Foreground, specifiedColor=<rgba 0 0 0 0>}
       cornerRadii=<8 8 8 8>
      }
      tintColor {colorUseRule=UseColor_Specified, specifiedColor=<rgba 0 0.0802 0.3515 1>}
      uVRegion {bIsValid=false, max=<0 0>, min=<0 0>}
     }
     checkedSlateSound {resourceObject=null}
     foregroundColor {colorUseRule=UseColor_Specified, specifiedColor=<rgba 1 1 1 1>}
     hoveredForeground {colorUseRule=UseColor_Specified, specifiedColor=<rgba 1 1 1 1>}
     hoveredSlateSound {resourceObject=null}
     padding=<ltrb 4 4 4 4>
     pressedForeground {colorUseRule=UseColor_Specified, specifiedColor=<rgba 1 1 1 1>}
     uncheckedHoveredImage {
      bIsDynamicallyLoaded=false
      drawAs=RoundedBox
      imageType=NoImage
      mirroring=NoMirror
      resourceName=None
      resourceObject=null
      tiling=NoTile
      imageSize=<0 0>
      margin=<ltrb 0 0 0 0>
      outlineSettings {
       bUseBrushTransparency=false
       roundingType=FixedRadius
       width=0
       color {colorUseRule=UseColor_Foreground, specifiedColor=<rgba 0 0 0 0>}
       cornerRadii=<8 8 8 8>
      }
      tintColor {colorUseRule=UseColor_Specified, specifiedColor=<rgba 0.0103 0.0103 0.0103 1>}
      uVRegion {bIsValid=false, max=<0 0>, min=<0 0>}
     }
     uncheckedImage {
      bIsDynamicallyLoaded=false
      drawAs=NoDrawType
      imageType=NoImage
      mirroring=NoMirror
      resourceName=None
      resourceObject=null
      tiling=NoTile
      imageSize=<0 0>
      margin=<ltrb 0 0 0 0>
      outlineSettings {
       bUseBrushTransparency=false
       roundingType=FixedRadius
       width=0
       color {colorUseRule=UseColor_Foreground, specifiedColor=<rgba 0 0 0 0>}
       cornerRadii=<4 4 4 4>
      }
      tintColor {colorUseRule=UseColor_Specified, specifiedColor=<rgba 0.0312 0.0312 0.0312 1>}
      uVRegion {bIsValid=false, max=<0 0>, min=<0 0>}
     }
     uncheckedPressedImage {
      bIsDynamicallyLoaded=false
      drawAs=RoundedBox
      imageType=NoImage
      mirroring=NoMirror
      resourceName=None
      resourceObject=null
      tiling=NoTile
      imageSize=<0 0>
      margin=<ltrb 0 0 0 0>
      outlineSettings {
       bUseBrushTransparency=false
       roundingType=FixedRadius
       width=0
       color {colorUseRule=UseColor_Foreground, specifiedColor=<rgba 0 0 0 0>}
       cornerRadii=<8 8 8 8>
      }
      tintColor {colorUseRule=UseColor_Specified, specifiedColor=<rgba 0.0625 0.0625 0.0625 1>}
      uVRegion {bIsValid=false, max=<0 0>, min=<0 0>}
     }
     uncheckedSlateSound {resourceObject=null}
     undeterminedForeground {colorUseRule=UseColor_Specified, specifiedColor=<rgba 1 1 1 1>}
     undeterminedHoveredImage {
      bIsDynamicallyLoaded=false
      drawAs=Image
      imageType=Vector
      mirroring=NoMirror
      resourceName=../../../Engine/Content/Slate/Starship/CoreWidgets/CheckBox/indeterminate.svg
      resourceObject=null
      tiling=NoTile
      imageSize=<16 16>
      margin=<ltrb 0 0 0 0>
      outlineSettings {
       bUseBrushTransparency=false
       roundingType=HalfHeightRadius
       width=0
       color {colorUseRule=UseColor_Specified, specifiedColor=<rgba 0 0 0 0>}
       cornerRadii=<0 0 0 0>
      }
      tintColor {colorUseRule=UseColor_Specified, specifiedColor=<rgba 0.0044 0.2384 1 1>}
      uVRegion {bIsValid=false, max=<0 0>, min=<0 0>}
     }
     undeterminedImage {
      bIsDynamicallyLoaded=false
      drawAs=Image
      imageType=Vector
      mirroring=NoMirror
      resourceName=../../../Engine/Content/Slate/Starship/CoreWidgets/CheckBox/indeterminate.svg
      resourceObject=null
      tiling=NoTile
      imageSize=<16 16>
      margin=<ltrb 0 0 0 0>
      outlineSettings {
       bUseBrushTransparency=false
       roundingType=HalfHeightRadius
       width=0
       color {colorUseRule=UseColor_Specified, specifiedColor=<rgba 0 0 0 0>}
       cornerRadii=<0 0 0 0>
      }
      tintColor {colorUseRule=UseColor_Specified, specifiedColor=<rgba 0 0.162 0.7454 1>}
      uVRegion {bIsValid=false, max=<0 0>, min=<0 0>}
     }
     undeterminedPressedImage {
      bIsDynamicallyLoaded=false
      drawAs=Image
      imageType=Vector
      mirroring=NoMirror
      resourceName=../../../Engine/Content/Slate/Starship/CoreWidgets/CheckBox/indeterminate.svg
      resourceObject=null
      tiling=NoTile
      imageSize=<16 16>
      margin=<ltrb 0 0 0 0>
      outlineSettings {
       bUseBrushTransparency=false
       roundingType=HalfHeightRadius
       width=0
       color {colorUseRule=UseColor_Specified, specifiedColor=<rgba 0 0 0 0>}
       cornerRadii=<0 0 0 0>
      }
      tintColor {colorUseRule=UseColor_Specified, specifiedColor=<rgba 0 0.0802 0.3515 1>}
      uVRegion {bIsValid=false, max=<0 0>, min=<0 0>}
     }
    }
   }
   slot {class=SizeBoxSlot, properties {horizontalAlignment=HAlign_Left, verticalAlignment=VAlign_Top}}
  }
  IconImage {
   widgetClassPath=&Image
   bIsVariable=true
   parent=ToggleButton
   widget=IconImage
   properties {
    brush {
     bIsDynamicallyLoaded=false
     drawAs=Image
     imageType=FullColor
     mirroring=NoMirror
     resourceName=None
     resourceObject=&Stars-Full-64
     tiling=NoTile
     imageSize=<24 24>
     margin=<ltrb 0 0 0 0>
     outlineSettings {
      bUseBrushTransparency=false
      roundingType=HalfHeightRadius
      width=0
      color {colorUseRule=UseColor_Specified, specifiedColor=<rgba 0 0 0 0>}
      cornerRadii=<0 0 0 0>
     }
     tintColor {colorUseRule=UseColor_Specified, specifiedColor=<rgba 1 1 1 1>}
     uVRegion {bIsValid=false, max=<0 0>, min=<0 0>}
    }
    colorAndOpacity=<rgba 0.5029 0.5029 0.5029 1>
   }
   slot {class=PanelSlot}
  }
 ]
}
