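# Editor utility list-entry widget for one dialogue line in the VO sound sequence editing tool.
# OnListItemObjectSet stores the FortDialogueData and initializes name, text, sound and delay controls; bindings show row striping, playback highlight and missing-sound warnings.
# Sound, delay and unassign edits broadcast through dispatchers; a SinglePropertyView bound to the widget itself edits Sound.
# Drag-over broadcasts line reordering, while Start/End Sound rows cannot be dragged.
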EditorUtilityWidgetBlueprint /Game/Sounds/Fort_Audio_Blueprints/Utilities/EditorUtilityWidgets/VO_Utilities/Widgets/FortSoundSequenceDialogueEntry
graphs [
 Delay Amount is Visible {nodes [
   Delay Amount is Visible() {is=title K2Node_FunctionEntry}
   n1 = Get ToggleButton!(self:Fort Sound Sequence Editor Toggle Button Object Reference=Get DelaySettingsToggleButton() -> (DelaySettingsToggleButton:Fort Sound Sequence Editor Toggle Button Object Reference)) -> (ToggleButton:Editor Utility Check Box Object Reference)
   > Return Node!(execute, ReturnValue:ESlateVisibility Enum=Select(Option 0=Hidden, Option 1=Visible, Index=IsChecked(self=@n1)))
  ]}
 EventGraph {nodes [
   n1 = "On Check State Changed (DelaySettingsToggleButton)"!() -> (OutputDelegate, then, Is Checked:bool) {is=title K2Node_ComponentBoundEvent}
   > n2 = Branch(Condition=@n1.Is Checked)
   Call On Delay Amount Changed!(
    execute:Exec=@n2.else
    self:Fort Sound Sequence Dialogue Entry Object Reference
    Dialogue Data Object:Fort Dialogue Data Object Reference=Get Dialogue Data Object() -> (Dialogue Data Object:Fort Dialogue Data Object Reference)
    New Delay Amount:double
   ) -> (then:Exec) {is=title K2Node_CallDelegate}
   "On Property Changed (SoundPropertyView)"!() -> (OutputDelegate, then, PropertyName:Name=None) {is=title K2Node_ComponentBoundEvent}
   > Call On Sound Changed!(
    execute:Exec
    self:Fort Sound Sequence Dialogue Entry Object Reference
    Dialogue Data Object:Fort Dialogue Data Object Reference=Get Dialogue Data Object()
    New Sound:Sound Base Object Reference=Get Sound() -> (Sound:Sound Base Object Reference)
   ) -> (then:Exec) {is=title K2Node_CallDelegate}
   "On Button Pressed (UnassignLineButton)"() {is=title K2Node_ComponentBoundEvent}
   > Call On Line Unassigned!(
    execute:Exec
    self:Fort Sound Sequence Dialogue Entry Object Reference
    Dialogue Data Object:Fort Dialogue Data Object Reference=Get Dialogue Data Object()
   ) -> (then:Exec) {is=title K2Node_CallDelegate}
   n3 = "On Value Changed (DelayAmountSpinbox)"!() -> (OutputDelegate, then, InValue:float) {is=title K2Node_ComponentBoundEvent}
   > Call On Delay Amount Changed(Dialogue Data Object=Get Dialogue Data Object(), New Delay Amount=@n3.InValue)
   Event PreConstruct()
   > SetObject(self=Get SoundPropertyView() -> (SoundPropertyView:Single Property View Object Reference), NewObject=Self-Reference())
   n4 = Event OnListItemObjectSet()
   > n5 = Cast To FortDialogueData!(execute, Object=@n4.ListItemObject) -> (then, CastFailed, AsFort Dialogue Data:Fort Dialogue Data Object Reference) {is=title K2Node_DynamicCast}
   > Set Dialogue Data Object(Dialogue Data Object:Fort Dialogue Data Object Reference=@n5) -> (Output_Get:Fort Dialogue Data Object Reference)
   > Initialize()
  ]}
 Get Background Color {nodes [
   Get Background Color() {is=title K2Node_FunctionEntry}
   n1 = IsHovered()
   n2 = Get Row Index!(self:Fort Dialogue Data Object Reference=Get Dialogue Data Object()) -> (Row Index:int=0)
   n3 = Select(
    Option 0:Linear Color Structure=(R=0.0103,G=0.0103,B=0.0103,A=1)
    Option 1:Linear Color Structure=(R=0.0075,G=0.0075,B=0.0075,A=1)
    Index="Equal (Integer)"(B="% (Integer)"(A=@n2, B=2) -> (ReturnValue=0))
   ) -> (ReturnValue:Linear Color Structure)
   n4 = Get Dialogue Data Object()
   > Return Node(ReturnValue:Linear Color Structure=Select(
     Option 0=Select(Option 0=@n3, Option 1=(R=0.0176,G=0.0176,B=0.0176,A=1), Index=@n1)
     Option 1=(R=0,G=0.162,B=0.7454,A=1)
     Index=Get Is Now Playing!(self:Fort Dialogue Data Object Reference=@n4) -> (Is Now Playing:bool)
    ))
  ]}
 Get Icon Brush {nodes [
   Get Icon Brush() {is=title K2Node_FunctionEntry}
   n1 = IsValid(Object=Get Sound())
   n2 = Get Character!(self:Fort Dialogue Data Object Reference=Get Dialogue Data Object()) -> (Character:String)
   n3 = Select(
    Option 0:Object Reference=/Game/Athena/UI/InGame/Creative/PropertyEditor/Icon_Picker/Textures_V2/64x64/T_UI_IconLibrary_SpeechBubble_64
    Option 1:Object Reference=/Game/Athena/UI/InGame/Creative/PropertyEditor/Icon_Picker/Textures_V2/64x64/T_UI_IconLibrary_Bell_64
    Index=OR Boolean(A="Equal Exactly (String)"(A=@n2, B=Start Sound), B="Equal Exactly (String)"(A=@n2, B=End Sound))
   ) -> (ReturnValue:Object Reference)
   > Return Node(ReturnValue:Slate Brush Structure=Make Slate Brush!(
     TintColor:Slate Color Structure=(SpecifiedColor=(R=1,G=1,B=1,A=1),ColorUseRule=UseColor_Specified)
     DrawAs:ESlateBrushDrawType Enum=Image
     Tiling:ESlateBrushTileType Enum=NoTile
     Mirroring:ESlateBrushMirrorType Enum=NoMirror
     ImageSize:Vector 2D Structure=(X=32,Y=32)
     Margin:Margin Structure=(Left=0,Top=0,Right=0,Bottom=0)
     ResourceObject:Object Reference=Select(Option 0=/FortniteCoreUI/HUD/MiniMap/Art/T_UI_Icon_Warning, Option 1=@n3, Index=@n1)
     OutlineSettings:Slate Brush Outline Settings Structure=(CornerRadii=(X=0,Y=0,Z=0,W=0),Color=(SpecifiedColor=(R=0,G=0,B=0,A=0),ColorUseRule=UseColor_Specified),Width=0,RoundingType=HalfHeightRadius,bUseBrushTransparency=False)
    ) -> (SlateBrush:Slate Brush Structure) {is=title K2Node_MakeStruct})
  ]}
 Get Sound Cue Asset Name {nodes [
   Get Sound Cue Asset Name() {is=title K2Node_FunctionEntry}
   n1 = Get Sound()
   > Return Node(ReturnValue:Text="To Text (String)"(InString=GetDisplayName(Object=@n1)))
  ]}
 Get Status Message Text {nodes [
   Get Status Message Text() {is=title K2Node_FunctionEntry}
   n1 = Get Sound()
   > Return Node(ReturnValue=Select(
     Option 0:Text=Format Text(Format="LOCTEXT(\"This line does not have a specified sound and will not be added to the sound sequence.\")")
     Option 1:Text
     Index=IsValid(Object=@n1)
    ) -> (ReturnValue:Text))
  ]}
 Get Status Message Text Color {nodes [
   Get Status Message Text Color() {is=title K2Node_FunctionEntry}
   n1 = Get Sound()
   > Return Node(ReturnValue:Slate Color Structure=Select(
     Option 0:Slate Color Structure=Make Slate Color(SpecifiedColor=(R=1,G=0.4793,B=0,A=1))
     Option 1:Slate Color Structure=Make Slate Color(SpecifiedColor=(R=0.5029,G=0.5029,B=0.5029,A=1))
     Index=IsValid(Object=@n1)
    ) -> (ReturnValue:Slate Color Structure))
  ]}
 Get Text Color {nodes [
   Get Text Color() {is=title K2Node_FunctionEntry}
   n1 = Get Sound()
   n2 = Select(
    Option 0:Linear Color Structure=(R=0.5029,G=0.5029,B=0.5029,A=1)
    Option 1:Linear Color Structure=(R=1,G=1,B=1,A=1)
    Index=Get Is Now Playing(self=Get Dialogue Data Object())
   ) -> (ReturnValue:Linear Color Structure)
   > Return Node(ReturnValue=Select(
     Option 0:Slate Color Structure=Make Slate Color(SpecifiedColor=(R=0.1,G=0.1,B=0.1,A=1))
     Option 1:Slate Color Structure=Make Slate Color(SpecifiedColor=@n2)
     Index=IsValid(Object=@n1)
    ) -> (ReturnValue:Slate Color Structure))
  ]}
 Get Toolbar Icon Color {nodes [
   Get Toolbar Icon Color() {is=title K2Node_FunctionEntry}
   n1 = Get Dialogue Data Object()
   > Return Node(ReturnValue:Linear Color Structure=Select(
     Option 0:Linear Color Structure=(R=0.5029,G=0.5029,B=0.5029,A=1)
     Option 1:Linear Color Structure=(R=1,G=1,B=1,A=1)
     Index=Get Is Now Playing(self=@n1)
    ) -> (ReturnValue:Linear Color Structure))
  ]}
 Get_ShowSoundCuePickerToggleButton_bIsEnabled {nodes [
   Get_ShowSoundCuePickerToggleButton_bIsEnabled() {is=title K2Node_FunctionEntry}
   n1 = Get Character(self=Get Dialogue Data Object())
   > Return Node(ReturnValue:bool=AND Boolean(A="Not Equal Exactly (String)"(A=@n1, B=Start Sound), B="Not Equal Exactly (String)"(A=@n1, B=End Sound)))
  ]}
 Get_SoundIcon_ColorAndOpacity {nodes [
   Get_SoundIcon_ColorAndOpacity() {is=title K2Node_FunctionEntry}
   n1 = Get Sound()
   n2 = Get Is Now Playing(self=Get Dialogue Data Object())
   > Return Node(ReturnValue:Linear Color Structure=Select(
     Option 0=(R=1,G=0.4793,B=0,A=1)
     Option 1=Select(Option 0=(R=0.5029,G=0.5029,B=0.5029,A=1), Option 1=(R=1,G=1,B=1,A=1), Index=@n2)
     Index=IsValid(Object=@n1)
    ))
  ]}
 Hide When Sound Cue Picker is Active {nodes [
   Hide When Sound Cue Picker is Active() {is=title K2Node_FunctionEntry}
   n1 = Get ToggleButton(self=Get ShowSoundCuePickerToggleButton() -> (ShowSoundCuePickerToggleButton:Fort Sound Sequence Editor Toggle Button Object Reference))
   > Return Node(ReturnValue:ESlateVisibility Enum=Select(Option 0:ESlateVisibility Enum=Visible, Option 1:ESlateVisibility Enum=Collapsed, Index=IsChecked(self=@n1)) -> (ReturnValue:ESlateVisibility Enum))
  ]}
 Initialize {nodes [
   Initialize() {is=title K2Node_FunctionEntry}
   > n1 = Sequence()
   n2 = Get Post-Sentence Delay!(self:Fort Dialogue Data Object Reference=Get Dialogue Data Object()) -> (Post-Sentence Delay:double)
   > SetCheckedState!(
    execute
    self:Check Box Object Reference=Get ToggleButton(self=Get DelaySettingsToggleButton())
    InCheckedState:ECheckBoxState Enum=Select(Option 0:ECheckBoxState Enum=Unchecked, Option 1:ECheckBoxState Enum=Checked, Index="float > float"(A=@n2)) -> (ReturnValue:ECheckBoxState Enum)
   ) -> (then)
   > Set Value!(
    execute
    Value:float=Get Post-Sentence Delay(self=Get Dialogue Data Object())
    self:Spin Box Object Reference=Get DelayAmountSpinbox() -> (DelayAmountSpinbox:Editor Utility Spin Box Object Reference)
   ) -> (then, Output_Get:float)
   n3 = Sequence(execute=@n1.then_1)
   n4 = Get Dialogue Data Object()
   > Set Text(
    Text="To Text (String)"(InString=Get Character(self=@n4))
    self=Get CharacterNameTextBlock() -> (CharacterNameTextBlock:Text Object Reference)
   )
   > Set Text(
    Text=Get Spoken Text!(self:Fort Dialogue Data Object Reference=Get Dialogue Data Object()) -> (Spoken Text:Text)
    self=Get SpokenTextBlock() -> (SpokenTextBlock:Text Object Reference)
   )
   Sequence(execute=@n3.then_1)
   > Set Sound(Sound:Sound Base Object Reference=Get Sound!(self:Fort Dialogue Data Object Reference=Get Dialogue Data Object()) -> (Sound)) -> (Output_Get:Sound Base Object Reference)
   > SetIsEnabled(
    self=Get UnassignLineButton() -> (UnassignLineButton:Fort Sound Sequence Editor Button Object Reference)
    bInIsEnabled=Get Enable Unassign Button!(self:Fort Dialogue Data Object Reference=Get Dialogue Data Object()) -> (Enable Unassign Button:bool)
   )
  ]}
 On Delay Amount Changed {nodes [On Delay Amount Changed!() -> (then, Dialogue Data Object:Fort Dialogue Data Object Reference, New Delay Amount:double=~) {is=title K2Node_FunctionEntry}]}
 On Line Unassigned {nodes [On Line Unassigned!() -> (then, Dialogue Data Object:Fort Dialogue Data Object Reference) {is=title K2Node_FunctionEntry}]}
 On Lines Reordered {nodes [On Lines Reordered!() -> (then, Dialogue Data to Insert:Fort Dialogue Data Object Reference, Dialogue Data to Insert Above:Fort Dialogue Data Object Reference) {is=title K2Node_FunctionEntry}]}
 On Sound Changed {nodes [On Sound Changed!() -> (then, Dialogue Data Object:Fort Dialogue Data Object Reference, New Sound:Sound Base Object Reference) {is=title K2Node_FunctionEntry}]}
 OnDragDetected {nodes [
   OnDragDetected!() -> (then, MyGeometry:Geometry Structure, PointerEvent:"Pointer Event Structure (by ref)") {is=title K2Node_FunctionEntry}
   n1 = Get Character(self=Get Dialogue Data Object())
   > Branch(Condition=AND Boolean(A="Not Equal Exactly (String)"(A=@n1, B=Start Sound), B="Not Equal Exactly (String)"(A=@n1, B=End Sound)))
   > n2 = "Create Drag && Drop Operation"!(
    execute:Exec
    WorldContextObject:Object Reference
    Class:Drag Drop Operation Class Reference=/Script/UMG.DragDropOperation
    Tag:String
    Payload:Object Reference=Get Dialogue Data Object()
    DefaultDragVisual:Widget Object Reference=Self-Reference()
    Pivot:EDragPivot Enum=TopLeft
    Offset:Vector 2D Structure=(X=0,Y=0)
   ) -> (then:Exec, ReturnValue:Drag Drop Operation Object Reference) {is=title K2Node_CreateDragDropOperation}
   > Return Node!(execute, Operation:Drag Drop Operation Object Reference=@n2)
  ]}
 OnDragOver {nodes [
   n1 = OnDragOver!() -> (then, MyGeometry:Geometry Structure, PointerEvent:Pointer Event Structure, Operation:Drag Drop Operation Object Reference) {is=title K2Node_FunctionEntry}
   > n2 = Cast To FortDialogueData(Object=Get Payload!(self:Drag Drop Operation Object Reference=@n1.Operation) -> (Payload:Object Reference))
   n3 = Get Dialogue Data Object()
   n4 = Get Character(self=@n3)
   > n5 = Branch(Condition=AND Boolean(
     A="Not Equal (Object)"(A:Object Reference=@n3, B:Object Reference=@n2) -> (ReturnValue:bool=~) {is=title K2Node_PromotableOperator}
     B="Not Equal Exactly (String)"(A=@n4, B=Start Sound)
    ))
   > Call On Lines Reordered!(
    execute:Exec
    self:Fort Sound Sequence Dialogue Entry Object Reference
    Dialogue Data to Insert:Fort Dialogue Data Object Reference=@n2
    Dialogue Data to Insert Above:Fort Dialogue Data Object Reference=Get Dialogue Data Object()
   ) -> (then:Exec) {is=title K2Node_CallDelegate}
   > Return Node!(execute, ReturnValue:bool=true)
   Return Node(execute=@n5.else, ReturnValue=true)
   Return Node(execute=@n2.CastFailed, ReturnValue=true)
  ]}
 Reset {nodes [
   Reset() {is=title K2Node_FunctionEntry}
   > Set Sound()
   > Set Dialogue Data Object()
   > Set Value(Value=0, self=Get DelayAmountSpinbox())
   > SetCheckedState(self=Get ToggleButton(self=Get ShowSoundCuePickerToggleButton()), InCheckedState=Unchecked)
   > SetCheckedState(self=Get ToggleButton(self=Get DelaySettingsToggleButton()), InCheckedState=Unchecked)
  ]}
 Show Sound Cue Picker {nodes [
   Show Sound Cue Picker() {is=title K2Node_FunctionEntry}
   n1 = Get ToggleButton(self=Get ShowSoundCuePickerToggleButton())
   > Return Node(ReturnValue:ESlateVisibility Enum=Select(Option 0:ESlateVisibility Enum=Collapsed, Option 1:ESlateVisibility Enum=Visible, Index=IsChecked(self=@n1)) -> (ReturnValue:ESlateVisibility Enum))
  ]}
]
variables [
 Dialogue Data Object {type=FortDialogueData_C, default=null}
 On Delay Amount Changed {default=()}
 On Line Unassigned {default=()}
 On Lines Reordered {default=()}
 On Sound Changed {default=()}
 Sound {type=SoundBase, default=null}
]
widgets {
 info {rootWidgetClass=&VerticalBox, widgetCount=22}
 widgets [
  VerticalBox_21 {widgetClassPath=&VerticalBox, widget=VerticalBox_21}
  HorizontalBox_35 {widgetClassPath=&HorizontalBox, parent=VerticalBox_21, widget=HorizontalBox_35, slot {class=VerticalBoxSlot}}
  Border {
   widgetClassPath=&Border
   bIsVariable=true
   parent=HorizontalBox_35
   widget=Border
   properties {
    background {
     bIsDynamicallyLoaded=false
     drawAs=Box
     imageType=NoImage
     mirroring=NoMirror
     resourceName=None
     resourceObject=null
     tiling=NoTile
     imageSize=<32 32>
     margin=<ltrb 0 0 0 0>
     outlineSettings {
      bUseBrushTransparency=false
      roundingType=FixedRadius
      width=2
      color {colorUseRule=UseColor_Specified, specifiedColor=<rgba 0 0 0 0>}
      cornerRadii=<0 0 0 0>
     }
     tintColor {colorUseRule=UseColor_Specified, specifiedColor=<rgba 1 1 1 1>}
     uVRegion {bIsValid=false, max=<0 0>, min=<0 0>}
    }
    brushColor=<rgba 0.0103 0.0103 0.0103 1>
    padding=<ltrb 4 4 4 4>
   }
   slot {class=HorizontalBoxSlot, properties {size {sizeRule=Fill, value=1}}}
  }
  VerticalBox_0 {widgetClassPath=&VerticalBox, parent=Border, widget=VerticalBox_0, slot {class=BorderSlot, properties {padding=<ltrb 4 4 4 4>}}}
  MainContent {widgetClassPath=&HorizontalBox, parent=VerticalBox_0, widget=MainContent, slot {class=VerticalBoxSlot}}
  IconBox {widgetClassPath=&VerticalBox, parent=MainContent, widget=IconBox, slot {class=HorizontalBoxSlot}}
  SizeBox_0 {
   widgetClassPath=&SizeBox
   parent=IconBox
   widget=SizeBox_0
   properties {bOverride_HeightOverride=true, bOverride_WidthOverride=true, heightOverride=32, widthOverride=32}
   slot {class=VerticalBoxSlot, properties {horizontalAlignment=HAlign_Center, verticalAlignment=VAlign_Center, padding=<ltrb 4 4 4 4>}}
  }
  SoundIcon {
   widgetClassPath=&Image
   bIsVariable=true
   parent=SizeBox_0
   widget=SoundIcon
   properties {
    brush {
     bIsDynamicallyLoaded=false
     drawAs=Image
     imageType=FullColor
     mirroring=NoMirror
     resourceName=None
     resourceObject=/Game/Athena/UI/InGame/Creative/PropertyEditor/Icon_Picker/Textures_V2/64x64/T_UI_IconLibrary_Chat_64
     tiling=NoTile
     imageSize=<64 64>
     margin=<ltrb 0 0 0 0>
     outlineSettings {
      bUseBrushTransparency=false
      roundingType=HalfHeightRadius
      width=0
      color {colorUseRule=UseColor_Specified, specifiedColor=<rgba 0 0 0 0>}
      cornerRadii=<0 0 0 0>
     }
     tintColor {colorUseRule=UseColor_Specified, specifiedColor=<rgba 1 1 1 1>}
     uVRegion {bIsValid=false, max=<0 0>, min=<0 0>}
    }
    colorAndOpacity=<rgba 0.5029 0.5029 0.5029 1>
   }
   slot {class=SizeBoxSlot}
  }
  Debug_Index {
   widgetClassPath=&TextBlock
   parent=IconBox
   widget=Debug_Index
   properties {
    text=Index
    visibility=Collapsed
    colorAndOpacity {colorUseRule=UseColor_Specified, specifiedColor=<rgba 0.5029 0.5029 0.5029 1>}
    font {
     bForceMonospaced=false
     bMaterialIsStencil=false
     fontMaterial=null
     fontObject=&Roboto
     letterSpacing=0
     monospacedWidth=1
     size=7.5
     skewAmount=0
     typefaceFontName=Regular
     outlineSettings {
      bApplyOutlineToDropShadows=false
      bMiteredCorners=false
      bSeparateFillAlpha=false
      outlineMaterial=null
      outlineSize=0
      outlineColor=<rgba 0 0 0 1>
     }
    }
   }
   slot {class=VerticalBoxSlot, properties {horizontalAlignment=HAlign_Center}}
  }
  BodyBox {
   widgetClassPath=&VerticalBox
   bIsVariable=true
   parent=MainContent
   widget=BodyBox
   properties {visibility=Visible}
   slot {class=HorizontalBoxSlot, properties {padding=<ltrb 4 4 4 2>, size {sizeRule=Fill, value=1}}}
  }
  DialogueBox {
   widgetClassPath=&VerticalBox
   parent=BodyBox
   widget=DialogueBox
   slot {class=VerticalBoxSlot, properties {size {sizeRule=Fill, value=1}}}
  }
  CharacterNameTextBlock {
   widgetClassPath=&TextBlock
   bIsVariable=true
   parent=DialogueBox
   widget=CharacterNameTextBlock
   properties {
    text=Character
    textTransformPolicy=ToUpper
    colorAndOpacity {colorUseRule=UseColor_Specified, specifiedColor=<rgba 0.5029 0.5029 0.5029 1>}
    font {
     bForceMonospaced=false
     bMaterialIsStencil=false
     fontMaterial=null
     fontObject=&Roboto
     letterSpacing=0
     monospacedWidth=1
     size=9.75
     skewAmount=0
     typefaceFontName=Bold
     outlineSettings {
      bApplyOutlineToDropShadows=false
      bMiteredCorners=false
      bSeparateFillAlpha=false
      outlineMaterial=null
      outlineSize=0
      outlineColor=<rgba 0 0 0 1>
     }
    }
   }
   slot {class=VerticalBoxSlot}
  }
  SpokenTextBox {
   widgetClassPath=&HorizontalBox
   parent=DialogueBox
   widget=SpokenTextBox
   slot {class=VerticalBoxSlot, properties {padding=<ltrb 0 4 0 0>}}
  }
  SpokenTextBlock {
   widgetClassPath=&TextBlock
   bIsVariable=true
   parent=SpokenTextBox
   widget=SpokenTextBlock
   properties {
    autoWrapText=true
    text=This is what the character said. It is a really long line that has to wrap on to a second line break.
    colorAndOpacity {colorUseRule=UseColor_Specified, specifiedColor=<rgba 0.5029 0.5029 0.5029 1>}
    font {
     bForceMonospaced=false
     bMaterialIsStencil=false
     fontMaterial=null
     fontObject=&Roboto
     letterSpacing=0
     monospacedWidth=1
     size=9.75
     skewAmount=0
     typefaceFontName=Regular
     outlineSettings {
      bApplyOutlineToDropShadows=false
      bMiteredCorners=false
      bSeparateFillAlpha=false
      outlineMaterial=null
      outlineSize=0
      outlineColor=<rgba 0 0 0 1>
     }
    }
   }
   slot {class=HorizontalBoxSlot, properties {size {sizeRule=Fill, value=1}}}
  }
  SoundPropertyView {
   widgetClassPath=&SinglePropertyView
   parent=MainContent
   widget=SoundPropertyView
   properties {bAutoLoadAsset=false, nameOverride=" ", propertyName=Sound}
   slot {class=HorizontalBoxSlot, properties {padding=<ltrb -12 2 2 2>, size {sizeRule=Fill, value=1}}}
  }
  RightTools {
   widgetClassPath=&HorizontalBox
   parent=MainContent
   widget=RightTools
   slot {class=HorizontalBoxSlot, properties {horizontalAlignment=HAlign_Right}}
  }
  Tools {
   widgetClassPath=&VerticalBox
   parent=RightTools
   widget=Tools
   slot {class=HorizontalBoxSlot, properties {horizontalAlignment=HAlign_Right, padding=<ltrb 0 0 4 0>}}
  }
  HorizontalBox_0 {
   widgetClassPath=&HorizontalBox
   parent=Tools
   widget=HorizontalBox_0
   slot {class=VerticalBoxSlot, properties {horizontalAlignment=HAlign_Right}}
  }
  ShowSoundCuePickerToggleButton {
   widgetClassPath=&/FortSoundSequenceEditorToggleButton
   bIsVariable=true
   parent=HorizontalBox_0
   widget=ShowSoundCuePickerToggleButton
   properties {icon=/Game/UI/Foundation/Textures/Icons/Audio/T-Icon-Sound-On-M, toolTip=Sound Cue, iconSize=<16 16>}
   slot {class=HorizontalBoxSlot, properties {padding=<ltrb 0 0 4 0>}}
  }
  DelaySettingsToggleButton {
   widgetClassPath=&FortSoundSequenceEditorToggleButton
   bIsVariable=true
   parent=HorizontalBox_0
   widget=DelaySettingsToggleButton
   properties {
    icon=/Game/Athena/UI/InGame/Creative/PropertyEditor/Icon_Picker/Textures_V2/64x64/T_UI_IconLibrary_Stopwatch_64
    toolTip=Delay Amount
    iconSize=<16 16>
   }
   slot {class=HorizontalBoxSlot, properties {padding=<ltrb 0 0 4 0>}}
  }
  UnassignLineButton {
   widgetClassPath=&/FortSoundSequenceEditorButton
   bIsVariable=true
   parent=HorizontalBox_0
   widget=UnassignLineButton
   properties {icon=/Game/UI/Foundation/Textures/Buttons/T-SlotIcon-NotRecyclable-64, toolTip=Unassign Line, iconSize=<16 16>}
   slot {class=HorizontalBoxSlot, properties {horizontalAlignment=HAlign_Right}}
  }
  DelayAmountSpinbox {
   widgetClassPath=/Script/Blutility.EditorUtilitySpinBox
   bIsVariable=true
   parent=Tools
   widget=DelayAmountSpinbox
   properties {
    bAlwaysUsesDeltaSnap=true
    bOverride_MaxSliderValue=true
    bOverride_MinSliderValue=true
    delta=0.1
    justification=Center
    maxFractionalDigits=2
    maxSliderValue=2
    font {
     bForceMonospaced=false
     bMaterialIsStencil=false
     fontMaterial=null
     fontObject=&Roboto
     letterSpacing=0
     monospacedWidth=1
     size=9
     skewAmount=0
     typefaceFontName=Bold
     outlineSettings {
      bApplyOutlineToDropShadows=false
      bMiteredCorners=false
      bSeparateFillAlpha=false
      outlineMaterial=null
      outlineSize=0
      outlineColor=<rgba 0 0 0 1>
     }
    }
    foregroundColor {colorUseRule=UseColor_Specified, specifiedColor=<rgba 0.5029 0.5029 0.5029 1>}
    widgetStyle {
     activeBackgroundBrush {
      bIsDynamicallyLoaded=false
      drawAs=RoundedBox
      imageType=NoImage
      mirroring=NoMirror
      resourceName=None
      resourceObject=null
      tiling=NoTile
      imageSize=<0 0>
      margin=<ltrb 0 0 0 0>
      outlineSettings {
       bUseBrushTransparency=false
       roundingType=FixedRadius
       width=1
       color {colorUseRule=UseColor_ColorTable, specifiedColor=<rgba 0 0 0 0>}
       cornerRadii=<4 4 4 4>
      }
      tintColor {colorUseRule=UseColor_Specified, specifiedColor=<rgba 0.0048 0.0048 0.0048 1>}
      uVRegion {bIsValid=false, max=<0 0>, min=<0 0>}
     }
     activeFillBrush {
      bIsDynamicallyLoaded=false
      drawAs=RoundedBox
      imageType=NoImage
      mirroring=NoMirror
      resourceName=None
      resourceObject=null
      tiling=NoTile
      imageSize=<0 0>
      margin=<ltrb 0 0 0 0>
      outlineSettings {
       bUseBrushTransparency=false
       roundingType=FixedRadius
       width=1
       color {colorUseRule=UseColor_Specified, specifiedColor=<rgba 0 0 0 0>}
       cornerRadii=<4 4 4 4>
      }
      tintColor {colorUseRule=UseColor_Specified, specifiedColor=<rgba 0 0.162 0.7454 1>}
      uVRegion {bIsValid=false, max=<0 0>, min=<0 0>}
     }
     arrowsImage {
      bIsDynamicallyLoaded=false
      drawAs=NoDrawType
      imageType=NoImage
      mirroring=NoMirror
      resourceName=None
      resourceObject=null
      tiling=NoTile
      imageSize=<0 0>
      margin=<ltrb 0 0 0 0>
      outlineSettings {
       bUseBrushTransparency=false
       roundingType=HalfHeightRadius
       width=0
       color {colorUseRule=UseColor_Specified, specifiedColor=<rgba 0 0 0 0>}
       cornerRadii=<0 0 0 0>
      }
      tintColor {colorUseRule=UseColor_Specified, specifiedColor=<rgba 1 1 1 1>}
      uVRegion {bIsValid=false, max=<0 0>, min=<0 0>}
     }
     backgroundBrush {
      bIsDynamicallyLoaded=false
      drawAs=RoundedBox
      imageType=NoImage
      mirroring=NoMirror
      resourceName=None
      resourceObject=null
      tiling=NoTile
      imageSize=<0 0>
      margin=<ltrb 0 0 0 0>
      outlineSettings {
       bUseBrushTransparency=false
       roundingType=FixedRadius
       width=1
       color {colorUseRule=UseColor_ColorTable, specifiedColor=<rgba 0 0 0 0>}
       cornerRadii=<4 4 4 4>
      }
      tintColor {colorUseRule=UseColor_Specified, specifiedColor=<rgba 0.0048 0.0048 0.0048 1>}
      uVRegion {bIsValid=false, max=<0 0>, min=<0 0>}
     }
     foregroundColor {colorUseRule=UseColor_Specified, specifiedColor=<rgba 1 1 1 1>}
     hoveredBackgroundBrush {
      bIsDynamicallyLoaded=false
      drawAs=RoundedBox
      imageType=NoImage
      mirroring=NoMirror
      resourceName=None
      resourceObject=null
      tiling=NoTile
      imageSize=<0 0>
      margin=<ltrb 0 0 0 0>
      outlineSettings {
       bUseBrushTransparency=false
       roundingType=FixedRadius
       width=1
       color {colorUseRule=UseColor_ColorTable, specifiedColor=<rgba 0 0 0 0>}
       cornerRadii=<4 4 4 4>
      }
      tintColor {colorUseRule=UseColor_Specified, specifiedColor=<rgba 0.0048 0.0048 0.0048 1>}
      uVRegion {bIsValid=false, max=<0 0>, min=<0 0>}
     }
     hoveredFillBrush {
      bIsDynamicallyLoaded=false
      drawAs=RoundedBox
      imageType=NoImage
      mirroring=NoMirror
      resourceName=None
      resourceObject=null
      tiling=NoTile
      imageSize=<0 0>
      margin=<ltrb 0 0 0 0>
      outlineSettings {
       bUseBrushTransparency=false
       roundingType=FixedRadius
       width=1
       color {colorUseRule=UseColor_Specified, specifiedColor=<rgba 0 0 0 0>}
       cornerRadii=<4 4 4 4>
      }
      tintColor {colorUseRule=UseColor_Specified, specifiedColor=<rgba 0.0953 0.0953 0.0953 1>}
      uVRegion {bIsValid=false, max=<0 0>, min=<0 0>}
     }
     inactiveFillBrush {
      bIsDynamicallyLoaded=false
      drawAs=RoundedBox
      imageType=NoImage
      mirroring=NoMirror
      resourceName=None
      resourceObject=null
      tiling=NoTile
      imageSize=<0 0>
      margin=<ltrb 0 0 0 0>
      outlineSettings {
       bUseBrushTransparency=false
       roundingType=FixedRadius
       width=1
       color {colorUseRule=UseColor_Specified, specifiedColor=<rgba 0 0 0 0>}
       cornerRadii=<4 4 4 4>
      }
      tintColor {colorUseRule=UseColor_Specified, specifiedColor=<rgba 0.0395 0.0395 0.0395 1>}
      uVRegion {bIsValid=false, max=<0 0>, min=<0 0>}
     }
     insetPadding=<ltrb 1 1 1 1>
     textPadding=<ltrb 0 1 0 1>
    }
   }
   slot {class=VerticalBoxSlot, properties {verticalAlignment=VAlign_Top, padding=<ltrb 0 4 0 0>, size {sizeRule=Fill, value=1}}}
  }
 ]
}
